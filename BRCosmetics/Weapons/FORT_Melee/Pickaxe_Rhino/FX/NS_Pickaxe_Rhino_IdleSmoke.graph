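# Niagara system for the Rhino pickaxe's idle exhaust, with socket-attached fire, flickering lights and smoke puffs.
# Packed emitters such as Fire_Packed and SmokeLoop_Packed merge several layers using spawn groups and switch helpers.
# Variant selectors, reused from the GateHound pickaxe, change color, size, offsets and spawning per cosmetic variant.
# User.bIsSwinging stops smoke spawning and suppresses one triggered fire burst, killing its live particles.
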NiagaraSystem /BRCosmetics/Weapons/FORT_Melee/Pickaxe_Rhino/FX/NS_Pickaxe_Rhino_IdleSmoke
dependencies {
 usedDataInterfaces [
  &NiagaraDataInterfacePlatformSet
  &NiagaraDataInterfaceArrayFloat
  &NiagaraDataInterfaceArrayInt32
  &NiagaraDataInterfaceArrayFloat2
  &NiagaraDataInterfaceCurve
  &NiagaraDataInterfaceVector2DCurve
  &NiagaraDataInterfaceVectorCurve
  &NiagaraDataInterfaceSpriteRendererInfo
  &NiagaraDataInterfaceSocketReader
 ]
 usedDynamicInputs [&Multiply_Vector2D_ByFloat, &RandomRangeVector2D]
 usedModules [
  &self:SystemSpawnScript.SystemScriptSource.SystemScriptGraph.NiagaraNodeAssignment_0.SetVariables_#1
  &SystemState
  &FN_VariantManager
 ]
}
emitters {
 Fire_Packed {
  emitterUpdateScript [
   &EmitterState(Inactive Response=::NewEnumerator1, Life Cycle Mode=::NewEnumerator1, Loop Duration=0.1, UseLoopDelay=false)
   &SpawnRate(
    Spawn Probability=/BRCosmetics/Weapons/FORT_Melee/Pickaxe_GateHound/FX/VariantManager/VariantFloatSelector(
     Original Float:NiagaraFloat=1
     Variant 1 Float:NiagaraFloat=0
     Variant 2 Float:NiagaraFloat=1
     Variant 3 Float:NiagaraFloat=1
     Variant Amount:NiagaraInt32=2
    )
    SpawnRate=&FN_PickFloatDependOnPlatform(TargetPlatformActive=10, TargetPlatformNoActive=5)
    Use Spawn Probability=true
   ) {module=SpawnRate001}
   &SpawnRate(
    Spawn Probability=&VariantFloatSelector(
     Original Float:NiagaraFloat=1
     Variant 1 Float:NiagaraFloat=0
     Variant 2 Float:NiagaraFloat=1
     Variant 3 Float:NiagaraFloat=1
     Variant Amount:NiagaraInt32=2
    )
    SpawnGroup=1
    SpawnRate=&FN_PickFloatDependOnPlatform(TargetPlatformActive=10, TargetPlatformNoActive=5)
    Use Spawn Probability=true
   )
   &SpawnBurst_Instantaneous(
    Spawn Count=&MakeIntFromFloat(Float=&VariantFloatSelector(
      Original Float:NiagaraFloat=2
      Variant 1 Float:NiagaraFloat=0
      Variant 2 Float:NiagaraFloat=5
      Variant 3 Float:NiagaraFloat=3
      Variant Amount:NiagaraInt32=2
     ))
    Spawn Group=2
    Spawn Probability=&VariantFloatSelector(
     Original Float:NiagaraFloat=1
     Variant 1 Float:NiagaraFloat=0
     Variant 2 Float:NiagaraFloat=1
     Variant 3 Float:NiagaraFloat=1
     Variant Amount:NiagaraInt32=2
    )
    Use Spawn Probability=true
    UseLoopCountLimit=false
   )
   /Game/Effects/Niagara/Modules/Spawn/FN_Spawn_TriggeredBurst(
    Cooldown=0.75
    Module.Use Burst Count Limit=false
    Spawn Count:NiagaraInt32=&Multiply_Int(
     A=&MakeIntFromFloat(Float=&VariantFloatSelector(
       Original Float:NiagaraFloat=2
       Variant 1 Float:NiagaraFloat=0
       Variant 2 Float:NiagaraFloat=2
       Variant 3 Float:NiagaraFloat=1
       Variant Amount:NiagaraInt32=2
      ))
     B=&MakeCustomIntFromBool(Bool=$User.bIsSwinging, True INT=0, False INT=1)
    )
    Spawn Group=3
    UseLoopCountLimit=false
   )
   &FN_Spawn_TriggeredBurst(
    Cooldown=0.75
    Module.Use Burst Count Limit=false
    Spawn Count:NiagaraInt32=&MakeIntFromFloat(Float=&VariantFloatSelector(
      Original Float:NiagaraFloat=0
      Variant 1 Float:NiagaraFloat=0
      Variant 2 Float:NiagaraFloat=5
      Variant 3 Float:NiagaraFloat=0
      Variant Amount:NiagaraInt32=2
     ))
    Spawn Group=4
    UseLoopCountLimit=false
   ) {module=FN_Spawn_TriggeredBurst001}
  ]
  particleSpawnScript [
   /SetVariables(Particles.Partition:NiagaraInt32=$Emitter.SpawnGroup) {isSetParameters=true}
   /SetVariables(Particles.SpawnGroup=&MergeHelper_IntSwitch(0=0, 1=0, 2=1, 3=2, 4=3, 5=?, Index=$Emitter.SpawnGroup, Num=::NewEnumerator3)) {isSetParameters=true}
   &InitializeParticle(
    Color=&MergeHelper_ColorSwitch(
     0=<rgba 1 1 1 0.06>
     1=&VariantColorSelector(
      Variant Amount=2
      Original Color=&RandomRangeLinearColor(Minimum=<rgba 1 1 1 0.025>, Maximum=<rgba 50 50 50 0.05>)
      Variant 1 Color=<rgba 0 0 0 0>
      Variant 2 Color=&MaskLinearColor(
       Mask Source=1
       Mask Parameter=/CascadeToNiagaraConverter/NiagaraScripts/DynamicInputScripts/SequentialInt(Use Array Data:NiagaraBool=false, First Number:NiagaraInt32=0, Sequence:NiagaraDataInterfaceArrayInt32=?, Last Number:NiagaraInt32=4)
       Mask=4
       Masked Linear Color=&RandomRangeLinearColor(Minimum=<rgba 75 5.113 0.633 0.1>, Maximum=<rgba 100 0.954 4.9339 0.1>)
       Passthrough Linear Color=&RandomRangeLinearColor(Minimum=<rgba 75 2.78 0 0.5>, Maximum=<rgba 100 0.954 4.934 0.5>)
      )
      Variant 3 Color=&MaskLinearColor(
       Mask Source=1
       Mask Parameter=&SequentialInt(Use Array Data:NiagaraBool=false, First Number:NiagaraInt32=0, Sequence:NiagaraDataInterfaceArrayInt32=?, Last Number:NiagaraInt32=2)
       Mask=2
       Masked Linear Color=&RandomRangeLinearColor(Minimum=<rgba 50 5 1.2 0.05>, Maximum=<rgba 70 8.8 1.8841 0.1>)
       Passthrough Linear Color=&RandomRangeLinearColor(Minimum=<rgba 75 4 1.2 0.05>, Maximum=<rgba 100 9 1.8841 0.1>)
      )
     )
     2=<rgba 0 0 0 0>
     3=<rgba 0 0 0 0>
     Index=$Particles.SpawnGroup
     Num=::NewEnumerator2
    )
    Lifetime=&MergeHelper_FloatSwitch(
     0=&RandomRangeFloat(Minimum=0.3, Maximum=0.4, Evaluation Type=::NewEnumerator1)
     1=&VariantFloatSelector(
      Original Float:NiagaraFloat=0.2
      Variant 1 Float:NiagaraFloat=0
      Variant 2 Float:NiagaraFloat=&MaskFloat(
       Mask Source=1
       Mask Parameter=&SequentialInt(Use Array Data:NiagaraBool=false, First Number:NiagaraInt32=0, Sequence:NiagaraDataInterfaceArrayInt32=?, Last Number:NiagaraInt32=4)
       Mask=4
       Masked Float=0.15
       Passthrough Float=&RandomRangeFloat(Minimum=0.25, Maximum=0.27, Evaluation Type=::NewEnumerator1)
      )
      Variant 3 Float:NiagaraFloat=&MaskFloat(
       Mask Source=1
       Mask Parameter=&SequentialInt(Use Array Data:NiagaraBool=false, First Number:NiagaraInt32=0, Sequence:NiagaraDataInterfaceArrayInt32=?, Last Number:NiagaraInt32=2)
       Mask=2
       Masked Float=0.2
       Passthrough Float=0.25
      )
      Variant Amount:NiagaraInt32=2
     )
     2=0.75
     3=1
     Index=$Particles.SpawnGroup
     Num=::NewEnumerator2
    )
    Mesh Scale=&MergeHelper_VectorSwitch(
     0=<0.135 0.625 0.625>
     1=<0.135 0.625 0.625>
     2=&MergeHelper_VectorSwitch(
      0=&RandomRangeVector(Minimum=<0.075 0.12 0.12>, Maximum=<0.075 0.15 0.15>, Evaluation Type=::NewEnumerator1)
      1=&RandomRangeVector(Minimum=<0.075 0.12 0.12>, Maximum=<0.075 0.15 0.15>, Evaluation Type=::NewEnumerator1)
      2=&RandomRangeVector(Minimum=<0.036 0.084 0.084>, Maximum=<0.06 0.108 0.108>, Evaluation Type=::NewEnumerator1)
      3=&RandomRangeVector(Minimum=<0.036 0.084 0.084>, Maximum=<0.06 0.108 0.108>, Evaluation Type=::NewEnumerator1)
      4=&RandomRangeVector(Minimum=<0.09 0.24 0.24>, Maximum=<0.15 0.3 0.3>, Evaluation Type=::NewEnumerator1)
      Index=&SequentialInt(Use Array Data:NiagaraBool=false, First Number:NiagaraInt32=0, Sequence:NiagaraDataInterfaceArrayInt32=?, Last Number:NiagaraInt32=4)
      Num=::NewEnumerator3
     )
     3=&MergeHelper_VectorSwitch(
      0=<0.08 0.18 0.18>
      1=<0.08 0.18 0.18>
      2=<0.15 0.22 0.22>
      3=?
      Index=&SequentialInt(Use Array Data:NiagaraBool=false, First Number:NiagaraInt32=0, Sequence:NiagaraDataInterfaceArrayInt32=?, Last Number:NiagaraInt32=2)
      Num=::NewEnumerator1
     )
     4=/BRCosmetics/Weapons/FORT_Melee/Pickaxe_GateHound/FX/VariantManager/VariantVectorSelector(
      Original Vector:Vector3f=<0.135 0.625 0.625>
      Variant 1 Vector:Vector3f=<0.135 0.625 0.625>
      Variant 2 Vector:Vector3f=&MergeHelper_VectorSwitch(
       0=&RandomRangeVector(Minimum=<0.075 0.12 0.12>, Maximum=<0.075 0.15 0.15>, Evaluation Type=::NewEnumerator1)
       1=&RandomRangeVector(Minimum=<0.075 0.12 0.12>, Maximum=<0.075 0.15 0.15>, Evaluation Type=::NewEnumerator1)
       2=&RandomRangeVector(Minimum=<0.036 0.084 0.084>, Maximum=<0.06 0.108 0.108>, Evaluation Type=::NewEnumerator1)
       3=&RandomRangeVector(Minimum=<0.036 0.084 0.084>, Maximum=<0.06 0.108 0.108>, Evaluation Type=::NewEnumerator1)
       4=&RandomRangeVector(Minimum=<0.09 0.24 0.24>, Maximum=<0.15 0.3 0.3>, Evaluation Type=::NewEnumerator1)
       Index=&SequentialInt(Use Array Data:NiagaraBool=false, First Number:NiagaraInt32=0, Sequence:NiagaraDataInterfaceArrayInt32=?, Last Number:NiagaraInt32=4)
       Num=::NewEnumerator3
      )
      Variant 3 Vector:Vector3f=&MergeHelper_VectorSwitch(
       0=<0.08 0.18 0.18>
       1=<0.08 0.18 0.18>
       2=<0.15 0.22 0.22>
       3=?
       Index=&SequentialInt(Use Array Data:NiagaraBool=false, First Number:NiagaraInt32=0, Sequence:NiagaraDataInterfaceArrayInt32=?, Last Number:NiagaraInt32=2)
       Num=::NewEnumerator1
      )
      Variant Amount:NiagaraInt32=2
     )
     Index=&MakeIntFromFloat(Float=$User.Variant, Conversion Mode=::NewEnumerator1)
     Num=::NewEnumerator2
    )
    Mesh Scale Mode=::NewEnumerator4
    Sprite Rotation Mode=::NewEnumerator2
    Sprite Size=&Vector2DFromFloat(Value=&VariantFloatSelector(
      Original Float:NiagaraFloat=&RandomRangeFloat(Minimum=10, Maximum=14, Evaluation Type=::NewEnumerator1)
      Variant 1 Float:NiagaraFloat=&RandomRangeFloat(Minimum=10, Maximum=14, Evaluation Type=::NewEnumerator1)
      Variant 2 Float:NiagaraFloat=&RandomRangeFloat(Minimum=10, Maximum=14, Evaluation Type=::NewEnumerator1)
      Variant 3 Float:NiagaraFloat=&RandomRangeFloat(Minimum=5.8, Maximum=7.8, Evaluation Type=::NewEnumerator1)
      Variant Amount:NiagaraInt32=2
     ))
    Sprite Size Mode=::NewEnumerator4
    UsePositionOffset=false
    Write Material Random=true
   )
   &AddVelocity(
    Cone Angle=&RandomRangeFloat(Minimum=0, Maximum=-0.02, Evaluation Type=::NewEnumerator1)
    Cone Angle Mode:ENiagara_AngleInput=::NewEnumerator1
    Cone Axis=<0 1 0>
    Module.UseConeFalloff=false
    Module.UseVelDistribution=false
    Override Seed=false
    Velocity Mode=::NewEnumerator2
    Velocity Speed=&MergeHelper_FloatSwitch(
     0=&VariantFloatSelector(
      Original Float:NiagaraFloat=&RandomRangeFloat(Minimum=-8, Maximum=-17, Evaluation Type=::NewEnumerator1)
      Variant 1 Float:NiagaraFloat=&RandomRangeFloat(Minimum=-8, Maximum=-17, Evaluation Type=::NewEnumerator1)
      Variant 2 Float:NiagaraFloat=&RandomRangeFloat(Minimum=-1, Maximum=-10, Evaluation Type=::NewEnumerator1)
      Variant 3 Float:NiagaraFloat=&RandomRangeFloat(Minimum=-8, Maximum=-17)
      Variant Amount:NiagaraInt32=2
     )
     1=0
     2=0
     3=?
     Index=$Particles.SpawnGroup
     Num=::NewEnumerator1
    )
   )
   &InitialMeshOrientation(Mesh Orientation Mode=::NewEnumerator2, Orientation Vector=<0 1 0>, Use Rotation Vector=false)
   &DynamicMaterialParameters(
    Dynamic Parameter Output:ENiagara_AttributeSamplingApplyOutput=::NewEnumerator1
    Index 0 Param 2=2
    Index 0 Param 3=20
    Index 0 Param 4=1
    Module.Index0Float0Write=true
    Module.Index0Float1Write=true
    Module.Index0Float2Write=true
    Module.Index0Float3Write=true
   ) {module=DynamicMaterialParameters001}
  ]
  particleUpdateScript [
   &ParticleState()
   &SocketLocation(
    Orientation Sampling=::NewEnumerator1
    SamplePosOffset=true
    Sampled Position Offset=&MergeHelper_VectorSwitch(
     0=&VariantVectorSelector(
      Original Vector:Vector3f=<12 0 0>
      Variant 1 Vector:Vector3f=<12 0 0>
      Variant 2 Vector:Vector3f=<-32 -0.5 16>
      Variant 3 Vector:Vector3f=&MergeHelper_VectorSwitch(
       0=<-20 -1 21>
       1=<-20 6.3 20.5>
       2=<-20 -8.3 20.5>
       3=?
       Index=&SequentialInt(Use Array Data:NiagaraBool=false, First Number:NiagaraInt32=0, Sequence:NiagaraDataInterfaceArrayInt32=?, Last Number:NiagaraInt32=2)
       Num=::NewEnumerator1
      )
      Variant Amount:NiagaraInt32=2
     )
     1=&VariantVectorSelector(
      Original Vector:Vector3f=<12 0 0>
      Variant 1 Vector:Vector3f=<12 0 0>
      Variant 2 Vector:Vector3f=&MergeHelper_VectorSwitch(
       0=<-31 -10.8 14.2>
       1=<-31 9.5 14.2>
       2=<-31 -10.9 9.4>
       3=<-31 9.5 9.4>
       4=<-31 -0.5 16>
       Index=&SequentialInt(Use Array Data:NiagaraBool=false, First Number:NiagaraInt32=0, Sequence:NiagaraDataInterfaceArrayInt32=?, Last Number:NiagaraInt32=4)
       Num=::NewEnumerator3
      )
      Variant 3 Vector:Vector3f=&MergeHelper_VectorSwitch(
       0=<-20 6.3 20.5>
       1=<-20 -8.5 20.5>
       2=<-20 -1 21>
       3=?
       Index=&SequentialInt(Use Array Data:NiagaraBool=false, First Number:NiagaraInt32=0, Sequence:NiagaraDataInterfaceArrayInt32=?, Last Number:NiagaraInt32=2)
       Num=::NewEnumerator1
      )
      Variant Amount:NiagaraInt32=2
     )
     2=&VariantVectorSelector(
      Original Vector:Vector3f=<18 0 0>
      Variant 1 Vector:Vector3f=<18 0 0>
      Variant 2 Vector:Vector3f=&MergeHelper_VectorSwitch(
       0=<-45 10 16>
       1=<-45 -10 16>
       2=?
       3=?
       Index=&SequentialInt(Use Array Data:NiagaraBool=false, First Number:NiagaraInt32=0, Sequence:NiagaraDataInterfaceArrayInt32=?, Last Number:NiagaraInt32=1)
       Num=::NewEnumerator0
      )
      Variant 3 Vector:Vector3f=<-30 0 19>
      Variant Amount:NiagaraInt32=2
     )
     3=&MergeHelper_VectorSwitch(
      0=<-32 -10.8 14.2>
      1=<-32 9.5 14.2>
      2=<-31 -10.9 9.4>
      3=<-31 9.5 9.4>
      4=<-31.5 -0.5 16>
      Index=&SequentialInt(Use Array Data:NiagaraBool=false, First Number:NiagaraInt32=0, Sequence:NiagaraDataInterfaceArrayInt32=?, Last Number:NiagaraInt32=4)
      Num=::NewEnumerator3
     )
     Index=$Particles.SpawnGroup
     Num=::NewEnumerator2
    )
    Socket Index=&MergeHelper_IntSwitch(
     0=0
     1=1
     2=&SequentialInt(Use Array Data:NiagaraBool=false, First Number:NiagaraInt32=0, Sequence:NiagaraDataInterfaceArrayInt32=?, Last Number:NiagaraInt32=1)
     3=&SequentialInt(Use Array Data:NiagaraBool=false, First Number:NiagaraInt32=0, Sequence:NiagaraDataInterfaceArrayInt32=?, Last Number:NiagaraInt32=1)
     4=&SequentialInt(Use Array Data:NiagaraBool=false, First Number:NiagaraInt32=0, Sequence:NiagaraDataInterfaceArrayInt32=?, Last Number:NiagaraInt32=1)
     5=?
     Index=$Particles.Partition
     Num=::NewEnumerator3
    )
    Socket Reader=$System.Socket Reader
    Socket Sampling Mode=::NewEnumerator3
    VelocitySpeedLimit=false
   )
   &ScaleMeshSize(Scale Factor=&VariantVectorSelector(
     Original Vector:Vector3f=&VectorFromCurve(VectorCurve=DataInterface() {propertyValues {
        XCurve {defaultValue=1, keys [<at 0 0.5 RCIM_Cubic>,<at 1 1.5 RCIM_Cubic>]}
        YCurve {keys [<at 0 1 RCIM_Cubic>]}
        ZCurve {keys [<at 0 1 RCIM_Cubic>,<at 1 0.9 RCIM_Cubic>]}
       }})
     Variant 1 Vector:Vector3f=<0 0 0>
     Variant 2 Vector:Vector3f=&VectorFromCurve(VectorCurve=DataInterface() {propertyValues {
        XCurve {keys [<at 0 0.75 RCIM_Cubic>,<at 1 2 RCIM_Cubic>]}
        YCurve {keys [<at 0 1 RCIM_Cubic>,<at 1 0.9 RCIM_Cubic>]}
        ZCurve {keys [<at 0 1 RCIM_Cubic>,<at 1 0.9 RCIM_Cubic>]}
       }})
     Variant 3 Vector:Vector3f=&VectorFromCurve(VectorCurve=DataInterface() {propertyValues {
        XCurve {defaultValue=1, keys [<at 0 0.5 RCIM_Cubic>,<at 1 1.5 RCIM_Cubic>]}
        YCurve {keys [<at 0 1 RCIM_Cubic>]}
        ZCurve {keys [<at 0 1 RCIM_Cubic>,<at 1 0.9 RCIM_Cubic>]}
       }})
     Variant Amount:NiagaraInt32=2
    ))
   &ScaleSpriteSize(
    Scale Sprite Size Mode=::NewEnumerator0
    Uniform Scale Factor=&VariantFloatSelector(
     Original Float:NiagaraFloat=&FloatFromCurve(FloatCurve=DataInterface() {propertyValues {Curve {keys [{arriveTangent=1.4511, interpMode=RCIM_Cubic, leaveTangent=1.4511, tangentMode=RCTM_User, time=-0.0013, value=0.1965}, <at 1 1.25>]}}})
     Variant 1 Float:NiagaraFloat=&FloatFromCurve(FloatCurve=DataInterface() {propertyValues {Curve {keys [{arriveTangent=1.4511, interpMode=RCIM_Cubic, leaveTangent=1.4511, tangentMode=RCTM_User, time=-0.0013, value=0.1965}, <at 1 1.25>]}}})
     Variant 2 Float:NiagaraFloat=&FloatFromCurve(FloatCurve=DataInterface() {propertyValues {Curve {keys [{arriveTangent=1.4511, interpMode=RCIM_Cubic, leaveTangent=1.4511, tangentMode=RCTM_User, time=0.0074, value=0.8625}, <at 0.9981 0.9955>]}}})
     Variant 3 Float:NiagaraFloat=&FloatFromCurve(FloatCurve=DataInterface() {propertyValues {Curve {keys [{arriveTangent=1.4511, interpMode=RCIM_Cubic, leaveTangent=1.4511, tangentMode=RCTM_User, time=-0.0013, value=0.1965}, <at 1 1.25>]}}})
     Variant Amount:NiagaraInt32=2
    )
   )
   &DynamicMaterialParameters(
    Dynamic Parameter Output:ENiagara_AttributeSamplingApplyOutput=::NewEnumerator1
    Index 0 Param 1=&FloatFromCurve(FloatCurve=DataInterface() {propertyValues {Curve {keys [<at 0.5 1 RCIM_Cubic>,<at 1 -1>]}}})
    Index 0 Param 2=1
    Index 0 Param 3=1
    Index 0 Param 4=1
    Module.Index0Float0Write=true
    Module.Index0Float1Write=true
    Module.Index0Float2Write=true
    Module.Index0Float3Write=true
   )
   &SubUVAnimation(
    Frames Per Second=10
    Playback Mode:ENiagara_FPSPlayrate=::NewEnumerator1
    Playback Scale Curve=DataInterface() {propertyValues {Curve {keys [
        {interpMode=RCIM_Cubic, tangentMode=RCTM_User, time=0.25, value=1}
        {arriveTangent=-1.2656, interpMode=RCIM_Cubic, leaveTangent=-1.2656, tangentMode=RCTM_User, time=1, value=0}
       ]}}}
    Random Start Frame=true
    "Sprite Renderer "=DataInterface() {propertyValues {SpriteRenderer=&self:Fire_Packed_0.NiagaraSpriteRendererProperties_2}}
    SubUV Animation Mode:ENiagara_SubUVLookupModeV2=::NewEnumerator3
    UseEndFrame=false
    UsePlayRateScale=false
    UseStartFrame=false
   )
   &Drag(Use Linear Drag=true, Use Rotational Drag=true)
   &ScaleColor(
    Scale Alpha=&MergeHelper_FloatSwitch(
     0=&FloatFromCurve(FloatCurve=DataInterface() {propertyValues {Curve {keys [
          <at 0 1 RCIM_Cubic>
          {arriveTangent=-0.573, interpMode=RCIM_Cubic, leaveTangent=-0.573, time=0.5987, value=0.9935}
          {arriveTangent=-3.008, interpMode=RCIM_Cubic, leaveTangent=-3.008, tangentMode=RCTM_User, time=0.856, value=0.5095}
          <at 1 0>
         ]}}})
     1=&FloatFromCurve(FloatCurve=DataInterface() {propertyValues {Curve {keys [<at 0 0 RCIM_Cubic>,<at 0.5076 0.4799 RCIM_Cubic>,<at 1 0 RCIM_Cubic>]}}})
     2=0
     3=?
     Index=$Particles.SpawnGroup
     Num=::NewEnumerator1
    )
    Scale RGB=&MergeHelper_VectorSwitch(
     0=&VariantVectorSelector(
      Original Vector:Vector3f=&VectorFromCurve(VectorCurve=DataInterface() {propertyValues {
         XCurve {keys [
           {arriveTangent=-861.5734, interpMode=RCIM_Cubic, leaveTangent=-861.5734, tangentMode=RCTM_User, time=0, value=100}
           <at 0.2668 50 RCIM_Cubic>
          ]}
         YCurve {keys [{arriveTangent=-66.82, interpMode=RCIM_Cubic, leaveTangent=-66.82, tangentMode=RCTM_User, time=0, value=0}, <at 0.2668 0 RCIM_Cubic>]}
         ZCurve {keys [{arriveTangent=-0.0001, interpMode=RCIM_Cubic, leaveTangent=-0.0001, tangentMode=RCTM_User, time=0, value=0}, <at 0.2668 0 RCIM_Cubic>]}
        }})
      Variant 1 Vector:Vector3f=&VectorFromCurve(VectorCurve=DataInterface() {propertyValues {
         XCurve {keys [
           {arriveTangent=-861.5734, interpMode=RCIM_Cubic, leaveTangent=-861.5734, tangentMode=RCTM_User, time=0, value=100}
           <at 0.2668 50 RCIM_Cubic>
          ]}
         YCurve {keys [{arriveTangent=-66.82, interpMode=RCIM_Cubic, leaveTangent=-66.82, tangentMode=RCTM_User, time=0, value=0}, <at 0.2668 0 RCIM_Cubic>]}
         ZCurve {keys [{arriveTangent=-0.0001, interpMode=RCIM_Cubic, leaveTangent=-0.0001, tangentMode=RCTM_User, time=0, value=0}, <at 0.2668 0 RCIM_Cubic>]}
        }})
      Variant 2 Vector:Vector3f=&VectorFromCurve(VectorCurve=DataInterface() {propertyValues {
         XCurve {keys [{arriveTangent=-861.5734, interpMode=RCIM_Cubic, leaveTangent=-861.5734, tangentMode=RCTM_User, time=0, value=50}, <at 0.2668 10 RCIM_Cubic>]}
         YCurve {keys [{arriveTangent=-66.82, interpMode=RCIM_Cubic, leaveTangent=-66.82, tangentMode=RCTM_User, time=0, value=1}, <at 0.2668 0.7939 RCIM_Cubic>]}
         ZCurve {keys [{arriveTangent=-0.0001, interpMode=RCIM_Cubic, leaveTangent=-0.0001, tangentMode=RCTM_User, time=0, value=1}, <at 0.2668 0 RCIM_Cubic>]}
        }})
      Variant 3 Vector:Vector3f=&VectorFromCurve(VectorCurve=DataInterface() {propertyValues {
         XCurve {keys [
           {arriveTangent=-861.5734, interpMode=RCIM_Cubic, leaveTangent=-861.5734, tangentMode=RCTM_User, time=0, value=100}
           <at 0.2668 50 RCIM_Cubic>
          ]}
         YCurve {keys [{arriveTangent=-66.82, interpMode=RCIM_Cubic, leaveTangent=-66.82, tangentMode=RCTM_User, time=0, value=0}, <at 0.2668 0 RCIM_Cubic>]}
         ZCurve {keys [{arriveTangent=-0.0001, interpMode=RCIM_Cubic, leaveTangent=-0.0001, tangentMode=RCTM_User, time=0, value=0}, <at 0.2668 0 RCIM_Cubic>]}
        }})
      Variant Amount:NiagaraInt32=2
     )
     1=<1 1 1>
     2=<0 0 0>
     3=?
     Index=$Particles.SpawnGroup
     Num=::NewEnumerator1
    )
    ScaleA=true
    ScaleRGB=true
   )
   &FN_Light_Attributes(
    Color=&VariantColorSelector(
     Variant Amount=2
     Original Color=<rgba 100 9.3613 1.36 1>
     Variant 1 Color=<rgba 100 9.3613 1.36 1>
     Variant 2 Color=<rgba 35 3.2765 0.476 1>
     Variant 3 Color=<rgba 100 9.3613 1.3641 1>
    )
    Create Color Attribute=true
    Create Exponent Attribute=true
    Create Radius Attribute=true
    Exponent=4
    Radius=20
    Scale Color=&FloatFromCurve(FloatCurve=DataInterface() {propertyValues {Curve {keys [
         {arriveTangent=-0.5378, time=0, value=1}
         {arriveTangent=0.759, leaveTangent=0.759, time=0.0565, value=0.4622}
         {arriveTangent=-0.1637, leaveTangent=-0.1637, time=0.1688, value=1.2212}
         {arriveTangent=-0.6961, leaveTangent=-0.6961, time=0.3486, value=1.0575}
         {arriveTangent=0.8156, leaveTangent=0.8156, time=0.4726, value=0.3614}
         {arriveTangent=0.2513, leaveTangent=0.2513, time=0.512, value=1.177}
         {arriveTangent=-1.2798, leaveTangent=-1.2798, time=0.5693, value=1.4283}
         {arriveTangent=1.0607, leaveTangent=1.0607, time=0.648, value=0.1484}
         {arriveTangent=-0.7391, leaveTangent=-0.7391, time=0.6833, value=1.2091}
         {arriveTangent=0.53, leaveTangent=0.53, time=0.7696, value=0.47}
         {arriveTangent=-0.8185, leaveTangent=-0.8185, time=0.8436, value=1}
         {arriveTangent=0.8185, leaveTangent=0.8185, time=0.953, value=0.1815}
         <at 1 1>
        ]}}})
   )
   &FN_Mesh_Attributes(
    Color=&MultiplyLinearColor(
     Linear Color=&MergeHelper_ColorSwitch(
      0=<rgba 25 5 1 0.5>
      1=<rgba 25 5 1 0.5>
      2=<rgba 25 5 1 0.5>
      3=<rgba 25 5 1 0.5>
      4=<rgba 25 1.4901 0 1>
      Index=&SequentialInt(Use Array Data:NiagaraBool=false, First Number:NiagaraInt32=0, Sequence:NiagaraDataInterfaceArrayInt32=?, Last Number:NiagaraInt32=4)
      Num=::NewEnumerator3
     )
     Scale Factor=&MakeVector4FromVectorAndFloat(
      "Vector (RGB)"=<1 1 1>
      "Float (Alpha)"=&FloatFromCurve(FloatCurve=DataInterface() {propertyValues {Curve {keys [
           {interpMode=RCIM_Cubic, leaveTangentWeight=0.1667, tangentMode=RCTM_User, time=0, value=0}
           {arriveTangentWeight=0.1667, interpMode=RCIM_Cubic, leaveTangentWeight=0.1667, tangentMode=RCTM_User, time=0.5, value=1}
           {arriveTangentWeight=0.1667, interpMode=RCIM_Cubic, tangentMode=RCTM_User, time=1, value=0}
          ]}}})
     )
    )
    Create Color Attribute=true
    Create Scale Attribute=true
    Module Output=::NewEnumerator1
    Scale=&MergeHelper_VectorSwitch(
     0=<3 3 3>
     1=<3 3 3>
     2=<3 3 3>
     3=<3 3 3>
     4=<5 5 5>
     Index=&SequentialInt(Use Array Data:NiagaraBool=false, First Number:NiagaraInt32=0, Sequence:NiagaraDataInterfaceArrayInt32=?, Last Number:NiagaraInt32=4)
     Num=::NewEnumerator3
    )
   )
   &SolveForcesAndVelocity(Clamp Velocity=false, Limit Acceleration=false)
   &KillParticles(Kill Particles=&Boolean_AndOperation(
     Bool1=&SetBoolByIntComparison(A=$Particles.Partition, B=3, Comparison Type=::NewEnumerator4)
     Bool2=$User.bIsSwinging
    ))
  ]
  rendererClasses [
   &NiagaraSpriteRendererProperties
   &NiagaraMeshRendererProperties
   &NiagaraLightRendererProperties
  ]
  renderers [
   {rendererClass=&NiagaraSpriteRendererProperties}
   {rendererClass=&NiagaraMeshRendererProperties, rendererIndex=1}
   {rendererClass=&NiagaraLightRendererProperties, rendererIndex=2}
   {rendererClass=&NiagaraMeshRendererProperties, rendererIndex=3}
  ]
 }
 SmokeLoop_Packed {
  emitterUpdateScript [
   &EmitterState(Inactive Response=::NewEnumerator1, Life Cycle Mode=::NewEnumerator1, Loop Duration=1, UseLoopDelay=true)
   &SpawnRate(
    Spawn Probability=&MakeCustomFloatFromBool(Bool=$User.bIsSwinging, True Float=0, False Float=1)
    SpawnRate=&VariantFloatSelector(
     Original Float:NiagaraFloat=11
     Variant 1 Float:NiagaraFloat=&FloatFromCurve(
      FloatCurve=DataInterface() {propertyValues {Curve {keys [<at 0 0 RCIM_Constant>,<at 0.6 1 RCIM_Constant>]}}}
      CurveIndex=$Emitter.NormalizedLoopAge
      Scale Curve=30
     )
     Variant 2 Float:NiagaraFloat=21
     Variant 3 Float:NiagaraFloat=21
     Variant Amount:NiagaraInt32=2
    )
    Use Spawn Probability=true
   )
   &SpawnRate(
    Spawn Probability=&MakeCustomFloatFromBool(Bool=$User.bIsSwinging, True Float=0, False Float=1)
    SpawnGroup=1
    SpawnRate=&VariantFloatSelector(
     Original Float:NiagaraFloat=11
     Variant 1 Float:NiagaraFloat=&FloatFromCurve(
      FloatCurve=DataInterface() {propertyValues {Curve {keys [<at 0 0 RCIM_Constant>,<at 0.6 1 RCIM_Constant>]}}}
      CurveIndex=$Emitter.NormalizedLoopAge
      Scale Curve=30
     )
     Variant 2 Float:NiagaraFloat=21
     Variant 3 Float:NiagaraFloat=21
     Variant Amount:NiagaraInt32=2
    )
    Use Spawn Probability=true
   ) {module=SpawnRate002}
   &SpawnRate(
    Spawn Probability=&MakeCustomFloatFromBool(Bool=$User.bIsSwinging, True Float=0, False Float=1)
    SpawnGroup=2
    SpawnRate=8
    Use Spawn Probability=true
   ) {module=SpawnRate001}
   &SpawnRate(
    Spawn Probability=&MakeCustomFloatFromBool(Bool=$User.bIsSwinging, True Float=0, False Float=1)
    SpawnGroup=3
    SpawnRate=8
    Use Spawn Probability=true
   ) {module=SpawnRate003}
  ]
  particleSpawnScript [
   /SetVariables(Particles.SpawnGroup=&MergeHelper_IntSwitch(0=0, 1=0, 2=1, 3=1, 4=?, 5=?, Index=$Emitter.SpawnGroup, Num=::NewEnumerator2)) {isSetParameters=true}
   &InitializeParticle(
    Color=&VariantColorSelector(
     Variant Amount=2
     Original Color=&MergeHelper_ColorSwitch(0=<rgba 0.1979 0.1979 0.1979 4>, 1=<rgba 0.1458 0.1458 0.1458 8>, 2=?, 3=?, Index=$Particles.SpawnGroup, Num=::NewEnumerator0)
     Variant 1 Color=<rgba 0.2305 0.2344 0.2267 6>
     Variant 2 Color=<rgba 0.241 0.2448 0.2368 8>
     Variant 3 Color=&MergeHelper_ColorSwitch(0=<rgba 0.21 0.2135 0.2066 4>, 1=<rgba 0.272 0.276 0.2671 8>, 2=?, 3=?, Index=$Particles.SpawnGroup, Num=::NewEnumerator0)
    )
    Lifetime=&VariantFloatSelector(
     Original Float:NiagaraFloat=&MergeHelper_FloatSwitch(
      0=&RandomRangeFloat(Minimum=0.3, Maximum=0.4, Evaluation Type=::NewEnumerator1)
      1=&RandomRangeFloat(Minimum=0.27, Maximum=0.3, Evaluation Type=::NewEnumerator1)
      2=&RandomRangeFloat(Minimum=0.3, Maximum=0.4, Evaluation Type=::NewEnumerator1)
      3=?
      Index=$Particles.SpawnGroup
      Num=::NewEnumerator0
     )
     Variant 1 Float:NiagaraFloat=&RandomRangeFloat(Minimum=0.35, Maximum=0.62, Evaluation Type=::NewEnumerator1)
     Variant 2 Float:NiagaraFloat=&RandomRangeFloat(Minimum=0.3, Maximum=0.4)
     Variant 3 Float:NiagaraFloat=&MergeHelper_FloatSwitch(
      0=&RandomRangeFloat(Minimum=0.3, Maximum=0.4, Evaluation Type=::NewEnumerator1)
      1=&RandomRangeFloat(Minimum=0.27, Maximum=0.3, Evaluation Type=::NewEnumerator1)
      2=?
      3=?
      Index=$Particles.SpawnGroup
      Num=::NewEnumerator0
     )
     Variant Amount:NiagaraInt32=2
    )
    Sprite Rotation Mode=::NewEnumerator2
    Sprite Size Mode=::NewEnumerator3
    Sprite UV Mode=::NewEnumerator3
    Uniform Sprite Size=&VariantFloatSelector(
     Original Float:NiagaraFloat=&MergeHelper_FloatSwitch(
      0=&RandomRangeFloat(Minimum=10, Maximum=11, Evaluation Type=::NewEnumerator1)
      1=&RandomRangeFloat(Minimum=6, Maximum=7.54, Evaluation Type=::NewEnumerator1)
      2=&RandomRangeFloat(Minimum=2, Maximum=5)
      3=?
      Index=$Particles.SpawnGroup
      Num=::NewEnumerator0
     )
     Variant 1 Float:NiagaraFloat=&RandomRangeFloat(Minimum=5, Maximum=7.5, Evaluation Type=::NewEnumerator1)
     Variant 2 Float:NiagaraFloat=&MergeHelper_FloatSwitch(
      0=&RandomRangeFloat(Minimum=5, Maximum=8)
      1=&RandomRangeFloat(Minimum=2, Maximum=5)
      2=?
      3=?
      Index=$Particles.SpawnGroup
      Num=::NewEnumerator0
     )
     Variant 3 Float:NiagaraFloat=&MergeHelper_FloatSwitch(
      0=&RandomRangeFloat(Minimum=10, Maximum=11)
      1=&RandomRangeFloat(Minimum=5, Maximum=6.2)
      2=?
      3=?
      Index=$Particles.SpawnGroup
      Num=::NewEnumerator0
     )
     Variant Amount:NiagaraInt32=2
    )
    UsePositionOffset=false
    Write Material Random=true
   )
   &SocketLocation(
    Orientation Sampling=::NewEnumerator1
    SamplePosOffset=true
    Sampled Position Offset=&VariantVectorSelector(
     Original Vector:Vector3f=<12 0 0>
     Variant 1 Vector:Vector3f=&MergeHelper_VectorSwitch(
      0=&Lerp_Vector(
       A=<30 -5 -8>
       B=<30 -5 8>
       Alpha=&MakeFloatFromInt(Int=&SequentialInt(Use Array Data:NiagaraBool=false, First Number:NiagaraInt32=0, Sequence:NiagaraDataInterfaceArrayInt32=?, Last Number:NiagaraInt32=1))
      )
      1=<10 0 0>
      2=?
      3=?
      Index=$Particles.SpawnGroup
      Num=::NewEnumerator0
     )
     Variant 2 Vector:Vector3f=&MergeHelper_VectorSwitch(
      0=&Lerp_Vector(
       A=<-32 15 12>
       B=<-34 -17 14>
       Alpha=&MakeFloatFromInt(Int=&SequentialInt(Use Array Data:NiagaraBool=false, First Number:NiagaraInt32=0, Sequence:NiagaraDataInterfaceArrayInt32=?, Last Number:NiagaraInt32=1))
      )
      1=&Lerp_Vector(
       A=<-13 -11 -4>
       B=<-13 9 -4>
       Alpha=&MakeFloatFromInt(Int=&SequentialInt(Use Array Data:NiagaraBool=false, First Number:NiagaraInt32=0, Sequence:NiagaraDataInterfaceArrayInt32=?, Last Number:NiagaraInt32=1))
      )
      2=?
      3=?
      Index=$Particles.SpawnGroup
      Num=::NewEnumerator0
     )
     Variant 3 Vector:Vector3f=&MergeHelper_VectorSwitch(
      0=&MergeHelper_VectorSwitch(
       0=<-24 9 20>
       1=<-24 -11 20>
       2=<-24 0 22>
       3=?
       Index=&SequentialInt(Use Array Data:NiagaraBool=false, First Number:NiagaraInt32=0, Sequence:NiagaraDataInterfaceArrayInt32=?, Last Number:NiagaraInt32=2)
       Num=::NewEnumerator1
      )
      1=&MergeHelper_VectorSwitch(
       0=<-14 -9 -5>
       1=<-14 9 -5>
       2=?
       3=?
       Index=&SequentialInt(Use Array Data:NiagaraBool=false, First Number:NiagaraInt32=0, Sequence:NiagaraDataInterfaceArrayInt32=?, Last Number:NiagaraInt32=1)
       Num=::NewEnumerator0
      )
      2=?
      3=?
      Index=$Particles.SpawnGroup
      Num=::NewEnumerator0
     )
     Variant Amount:NiagaraInt32=2
    )
    Socket Index=$Emitter.SpawnGroup
    Socket Reader=$System.Socket Reader
    Socket Sampling Mode=::NewEnumerator3
    VelocitySpeedLimit=false
   )
   &AddVelocity(Velocity=&VariantVectorSelector(
     Original Vector:Vector3f=<0 -60 0>
     Variant 1 Vector:Vector3f=&MergeHelper_VectorSwitch(
      0=&Lerp_Vector(
       A=&RandomRangeVector(Minimum=<-40 -45 -6>, Maximum=<-12 -20 -2>, Evaluation Type=::NewEnumerator1)
       B=&RandomRangeVector(Minimum=<40 -45 -6>, Maximum=<12 -20 -2>, Evaluation Type=::NewEnumerator1)
       Alpha=&ClampFloat(Float=&MakeFloatFromInt(Int=$Emitter.SpawnGroup))
      )
      1=&RandomRangeVector(Minimum=<0 -40 -15>, Maximum=<0 -20 -8>, Evaluation Type=::NewEnumerator1)
      2=?
      3=?
      Index=$Particles.SpawnGroup
      Num=::NewEnumerator0
     )
     Variant 2 Vector:Vector3f=&RandomRangeVector(Minimum=<0 -40 -20>, Maximum=<0 -30 -10>, Evaluation Type=::NewEnumerator1)
     Variant 3 Vector:Vector3f=&MergeHelper_VectorSwitch(
      0=&RandomRangeVector(Minimum=<0 -40 -30>, Maximum=<0 -30 -20>, Evaluation Type=::NewEnumerator1)
      1=&RandomRangeVector(Minimum=<0 -20 -15>, Maximum=<0 -15 -10>, Evaluation Type=::NewEnumerator1)
      2=?
      3=?
      Index=$Particles.SpawnGroup
      Num=::NewEnumerator0
     )
     Variant Amount:NiagaraInt32=2
    ))
   &InheritVelocity(
    Inherited Velocity Amount Scale=&VariantVectorSelector(
     Original Vector:Vector3f=&MergeHelper_VectorSwitch(0=<0.6 0.6 0.2>, 1=<0.5 0.5 0.2>, 2=?, 3=?, Index=$Particles.SpawnGroup, Num=::NewEnumerator0)
     Variant 1 Vector:Vector3f=<0.85 0.85 0.3>
     Variant 2 Vector:Vector3f=<0.6 0.6 0.2>
     Variant 3 Vector:Vector3f=&MergeHelper_VectorSwitch(0=<0.85 0.85 0.2>, 1=<0.8 0.8 0.2>, 2=?, 3=?, Index=$Particles.SpawnGroup, Num=::NewEnumerator0)
     Variant Amount:NiagaraInt32=2
    )
    Inherited Velocity Speed Limit=0
    Speed Limit=false
    UseSpeedThreshold=false
   )
  ]
  particleUpdateScript [
   &ParticleState()
   &ScaleSpriteSize(
    Scale Sprite Size Mode=::NewEnumerator0
    Uniform Scale Factor=&VariantFloatSelector(
     Original Float:NiagaraFloat=&FloatFromCurve(FloatCurve=DataInterface() {propertyValues {Curve {keys [
          {arriveTangent=2.7132, interpMode=RCIM_Cubic, leaveTangent=2.7132, tangentMode=RCTM_User, time=0, value=1}
          {arriveTangent=2.4, interpMode=RCIM_Cubic, leaveTangent=2.4, time=0.2677, value=1.8}
          <at 1 3.4 RCIM_Cubic>
         ]}}})
     Variant 1 Float:NiagaraFloat=&FloatFromCurve(FloatCurve=DataInterface() {propertyValues {Curve {keys [
          {arriveTangent=2.7132, interpMode=RCIM_Cubic, leaveTangent=2.7132, tangentMode=RCTM_User, time=0, value=1}
          {arriveTangent=2.4, interpMode=RCIM_Cubic, leaveTangent=2.4, time=0.2677, value=1.8}
          <at 1 3.4 RCIM_Cubic>
         ]}}})
     Variant 2 Float:NiagaraFloat=&MergeHelper_FloatSwitch(
      0=&FloatFromCurve(FloatCurve=DataInterface() {propertyValues {Curve {keys [
           {arriveTangent=2.7132, interpMode=RCIM_Cubic, leaveTangent=2.7132, tangentMode=RCTM_User, time=-0.0018, value=0.993}
           {arriveTangent=1.5228, interpMode=RCIM_Cubic, leaveTangent=1.5228, time=0.252, value=2.0091}
           <at 0.9969 2.5136 RCIM_Cubic>
          ]}}})
      1=&FloatFromCurve(FloatCurve=DataInterface() {propertyValues {Curve {keys [
           {arriveTangent=2.7132, interpMode=RCIM_Cubic, leaveTangent=2.7132, tangentMode=RCTM_User, time=0, value=1}
           {arriveTangent=2.4, interpMode=RCIM_Cubic, leaveTangent=2.4, time=0.2677, value=1.8}
           <at 1 3.4 RCIM_Cubic>
          ]}}})
      2=?
      3=?
      Index=$Particles.SpawnGroup
      Num=::NewEnumerator0
     )
     Variant 3 Float:NiagaraFloat=&FloatFromCurve(FloatCurve=DataInterface() {propertyValues {Curve {keys [
          {arriveTangent=2.7132, interpMode=RCIM_Cubic, leaveTangent=2.7132, tangentMode=RCTM_User, time=0, value=1}
          {arriveTangent=2.4, interpMode=RCIM_Cubic, leaveTangent=2.4, time=0.2677, value=1.8}
          <at 1 3.4 RCIM_Cubic>
         ]}}})
     Variant Amount:NiagaraInt32=2
    )
   )
   &ScaleColor(
    Scale Alpha=&VariantFloatSelector(
     Original Float:NiagaraFloat=&FloatFromCurve(FloatCurve=DataInterface() {propertyValues {Curve {keys [
          {arriveTangent=5.9578, interpMode=RCIM_Cubic, leaveTangent=5.9578, tangentMode=RCTM_User, time=0.0008, value=0.019}
          {arriveTangent=0.1001, interpMode=RCIM_Cubic, leaveTangent=0.1001, tangentMode=RCTM_User, time=0.1125, value=0.9699}
          {arriveTangent=-1.0929, interpMode=RCIM_Cubic, leaveTangent=-1.0929, time=0.3315, value=0.5264}
          <at 1 0>
         ]}}})
     Variant 1 Float:NiagaraFloat=&FloatFromCurve(FloatCurve=DataInterface() {propertyValues {Curve {keys [
          {arriveTangent=5.9578, interpMode=RCIM_Cubic, leaveTangent=5.9578, tangentMode=RCTM_User, time=0.0008, value=0.019}
          {arriveTangent=0.1001, interpMode=RCIM_Cubic, leaveTangent=0.1001, tangentMode=RCTM_User, time=0.1125, value=0.9699}
          {arriveTangent=-1.0929, interpMode=RCIM_Cubic, leaveTangent=-1.0929, time=0.3315, value=0.5264}
          <at 1 0>
         ]}}})
     Variant 2 Float:NiagaraFloat=&FloatFromCurve(FloatCurve=DataInterface() {propertyValues {Curve {keys [
          {arriveTangent=5.9578, interpMode=RCIM_Cubic, leaveTangent=5.9578, tangentMode=RCTM_User, time=0.0008, value=0.019}
          {arriveTangent=0.1001, interpMode=RCIM_Cubic, leaveTangent=0.1001, tangentMode=RCTM_User, time=0.1245, value=0.4654}
          {arriveTangent=-0.5316, interpMode=RCIM_Cubic, leaveTangent=-0.5316, time=0.3304, value=0.3952}
          <at 1 0>
         ]}}})
     Variant 3 Float:NiagaraFloat=&FloatFromCurve(FloatCurve=DataInterface() {propertyValues {Curve {keys [
          {arriveTangent=5.9578, interpMode=RCIM_Cubic, leaveTangent=5.9578, tangentMode=RCTM_User, time=0.0008, value=0.019}
          {arriveTangent=0.1001, interpMode=RCIM_Cubic, leaveTangent=0.1001, tangentMode=RCTM_User, time=0.1125, value=0.9699}
          {arriveTangent=-1.0929, interpMode=RCIM_Cubic, leaveTangent=-1.0929, time=0.3315, value=0.5264}
          <at 1 0>
         ]}}})
     Variant Amount:NiagaraInt32=2
    )
    Scale RGB=&VariantVectorSelector(
     Original Vector:Vector3f=&VectorFromCurve(VectorCurve=DataInterface() {propertyValues {XCurve {keys [<at 0 0.8125>,<at 1 0.6406>]}, YCurve {keys [<at 0 0.8125>,<at 1 0.6406>]}, ZCurve {keys [<at 0 0.8125>,<at 1 0.6406>]}}})
     Variant 1 Vector:Vector3f=&VectorFromCurve(VectorCurve=DataInterface() {propertyValues {XCurve {keys [<at 0 0.4583>,<at 1 1>]}, YCurve {keys [<at 0 0.4583>,<at 1 1>]}, ZCurve {keys [<at 0 0.4583>,<at 1 1>]}}})
     Variant 2 Vector:Vector3f=&VectorFromCurve(VectorCurve=DataInterface() {propertyValues {XCurve {keys [<at 0 0.2708>,<at 1 0.9844>]}, YCurve {keys [<at 0 0.2708>,<at 1 0.9844>]}, ZCurve {keys [<at 0 0.2708>,<at 1 0.9844>]}}})
     Variant 3 Vector:Vector3f=&VectorFromCurve(VectorCurve=DataInterface() {propertyValues {XCurve {keys [<at 0 0.2708>,<at 1 0.9844>]}, YCurve {keys [<at 0 0.2708>,<at 1 0.9844>]}, ZCurve {keys [<at 0 0.2708>,<at 1 0.9844>]}}})
     Variant Amount:NiagaraInt32=2
    )
    ScaleA=true
    ScaleRGB=true
   )
   &DynamicMaterialParameters(
    Index 0 Param 1=&FloatFromCurve(FloatCurve=DataInterface() {propertyValues {Curve {keys [<at 0.3 1 RCIM_Cubic>,<at 1 0>]}}})
    Index 0 Param 2=&FloatFromCurve(FloatCurve=DataInterface() {propertyValues {Curve {keys [<at 0 1 RCIM_Cubic>,<at 0.8 8.7408 RCIM_Cubic>]}}})
    Index 0 Param 3=1
    Index 0 Param 4=&FloatFromCurve(FloatCurve=DataInterface() {propertyValues {Curve {keys [<at 0.3 0>,<at 0.5 0.5>]}}})
    Module.Index0Float0Write=true
    Module.Index0Float1Write=true
    Module.Index0Float2Write=true
    Module.Index0Float3Write=true
   )
   &SubUVAnimation(
    Frames Per Second=25
    Playback Mode:ENiagara_FPSPlayrate=::NewEnumerator1
    Playback Scale Curve=DataInterface() {propertyValues {Curve {keys [
        {interpMode=RCIM_Cubic, tangentMode=RCTM_User, time=0.25, value=1}
        {arriveTangent=-1.2656, interpMode=RCIM_Cubic, leaveTangent=-1.2656, tangentMode=RCTM_User, time=1, value=0}
       ]}}}
    Random Start Frame=true
    "Sprite Renderer "=DataInterface() {propertyValues {SpriteRenderer=&self:SmokeLoop_Packed_0.NiagaraSpriteRendererProperties_1}}
    SubUV Animation Mode:ENiagara_SubUVLookupModeV2=::NewEnumerator3
    UseEndFrame=false
    UsePlayRateScale=false
    UseStartFrame=false
   )
   &Drag(
    Drag=&VariantFloatSelector(
     Original Float:NiagaraFloat=&MergeHelper_FloatSwitch(0=1.2, 1=1.5, 2=?, 3=?, Index=$Particles.SpawnGroup, Num=::NewEnumerator0)
     Variant 1 Float:NiagaraFloat=1.5
     Variant 2 Float:NiagaraFloat=1.2
     Variant 3 Float:NiagaraFloat=&MergeHelper_FloatSwitch(0=1.2, 1=1.5, 2=?, 3=?, Index=$Particles.SpawnGroup, Num=::NewEnumerator0)
     Variant Amount:NiagaraInt32=2
    )
    Use Linear Drag=true
    Use Rotational Drag=true
   )
   &SolveForcesAndVelocity(Clamp Velocity=false, Limit Acceleration=false)
  ]
  rendererClasses [&NiagaraSpriteRendererProperties]
  renderers [{rendererClass=&NiagaraSpriteRendererProperties}]
 }
 SwingParticles_Local {
  emitterUpdateScript [
   &EmitterState(Inactive Response=::NewEnumerator1, Life Cycle Mode=::NewEnumerator1, Loop Duration=0.1, UseLoopDelay=false)
   &SpawnBurst_Instantaneous(
    Spawn Count=&MakeIntFromFloat(Float=&VariantFloatSelector(
      Original Float:NiagaraFloat=2
      Variant 1 Float:NiagaraFloat=0
      Variant 2 Float:NiagaraFloat=5
      Variant 3 Float:NiagaraFloat=3
      Variant Amount:NiagaraInt32=2
     ))
    Spawn Probability=&MakeCustomFloatFromBool(
     Bool=&Boolean_AndOperation(
      Bool1=$User.bIsSwinging
      Bool2=&SetBoolByIntComparison(
       A=&MakeIntFromFloat(Float=&VariantFloatSelector(
         Original Float:NiagaraFloat=1
         Variant 1 Float:NiagaraFloat=0
         Variant 2 Float:NiagaraFloat=1
         Variant 3 Float:NiagaraFloat=1
         Variant Amount:NiagaraInt32=2
        ))
       B=1
       Comparison Type=::NewEnumerator4
      )
     )
     True Float=1
     False Float=0
    )
    Use Spawn Probability=true
    UseLoopCountLimit=false
   )
   &SpawnRate(
    Spawn Probability=&MakeCustomFloatFromBool(
     Bool=&Boolean_AndOperation(
      Bool1=$User.bIsSwinging
      Bool2=&SetBoolByIntComparison(
       A=&MakeIntFromFloat(Float=&VariantFloatSelector(
         Original Float:NiagaraFloat=1
         Variant 1 Float:NiagaraFloat=0
         Variant 2 Float:NiagaraFloat=1
         Variant 3 Float:NiagaraFloat=1
         Variant Amount:NiagaraInt32=2
        ))
       B=1
       Comparison Type=::NewEnumerator4
      )
     )
     True Float=1
     False Float=0
    )
    SpawnGroup=1
    SpawnRate=&FN_PickFloatDependOnPlatform(TargetPlatformActive=100, TargetPlatformNoActive=15)
    Use Spawn Probability=true
   ) {module=SpawnRate001}
   &FN_Spawn_TriggeredBurst(
    Cooldown:NiagaraFloat=&RandomRangeFloat(Minimum=0.6, Maximum=1.5)
    Module.Use Burst Count Limit=false
    Spawn Count:NiagaraInt32=&Multiply_Int(
     A=&RandomRangeInt(Minimum=2, Maximum=4)
     B=&MakeIntFromFloat(Float=&VariantFloatSelector(
       Original Float:NiagaraFloat=0
       Variant 1 Float:NiagaraFloat=1
       Variant 2 Float:NiagaraFloat=0
       Variant 3 Float:NiagaraFloat=0
       Variant Amount:NiagaraInt32=2
      ))
    )
    Spawn Group=2
    UseLoopCountLimit=false
   )
   &SpawnRate(
    Spawn Probability=&VariantFloatSelector(
     Original Float:NiagaraFloat=0
     Variant 1 Float:NiagaraFloat=1
     Variant 2 Float:NiagaraFloat=0
     Variant 3 Float:NiagaraFloat=0
     Variant Amount:NiagaraInt32=2
    )
    SpawnGroup=2
    SpawnRate=&FN_PickFloatDependOnPlatform(TargetPlatformActive=10, TargetPlatformNoActive=4)
    Use Spawn Probability=true
   ) {module=SpawnRate002}
  ]
  particleSpawnScript [
   /SetVariables(Particles.VisibilityTag=$Emitter.SpawnGroup) {isSetParameters=true}
   &InitializeParticle(
    Lifetime=&MergeHelper_FloatSwitch(
     0=0.3
     1=&RandomRangeFloat(Minimum=0.1, Maximum=0.15, Evaluation Type=::NewEnumerator1)
     2=&RandomRangeFloat(Minimum=0.18, Maximum=0.45, Evaluation Type=::NewEnumerator1)
     3=?
     Index=$Particles.VisibilityTag
     Num=::NewEnumerator1
    )
    Mesh Scale=&VariantVectorSelector(
     Original Vector:Vector3f=&RandomRangeVector(Minimum=<0.32 0.32 0.32>, Maximum=<0.42 0.32 0.32>)
     Variant 1 Vector:Vector3f=<0 0 0>
     Variant 2 Vector:Vector3f=&MaskVector(
      Mask Source=1
      Mask Parameter=&SequentialInt(Use Array Data:NiagaraBool=false, First Number:NiagaraInt32=0, Sequence:NiagaraDataInterfaceArrayInt32=?, Last Number:NiagaraInt32=4)
      Mask=4
      Masked Vector=<0.7 0.3 0.3>
      Passthrough Vector=&RandomRangeVector(Minimum=<0.4 0.2 0.2>, Maximum=<0.6 0.2 0.2>)
     )
     Variant 3 Vector:Vector3f=&MergeHelper_VectorSwitch(
      0=<0.4 0.1 0.1>
      1=<0.4 0.1 0.1>
      2=<0.55 0.15 0.15>
      3=?
      Index=&SequentialInt(Use Array Data:NiagaraBool=false, First Number:NiagaraInt32=0, Sequence:NiagaraDataInterfaceArrayInt32=?, Last Number:NiagaraInt32=2)
      Num=::NewEnumerator1
     )
     Variant Amount:NiagaraInt32=2
    )
    Mesh Scale Mode=::NewEnumerator4
    Sprite Rotation Mode=::NewEnumerator2
    Sprite Size=/Niagara/DynamicInputs/Helpers/MaskVector2d(
     Mask Source:NiagaraInt32=1
     Mask Parameter:NiagaraInt32=$Particles.VisibilityTag
     Mask:NiagaraInt32=2
     Masked Vector2D:Vector2f=&Vector2DFromFloat(Value=&RandomRangeFloat(Minimum=1.5, Maximum=3))
     Passthrough Vector2D:Vector2f=&Vector2DFromFloat(Value=&RandomRangeFloat(Minimum=2.35, Maximum=4.5))
    )
    Sprite Size Mode=::NewEnumerator4
    Sprite UV Mode=::NewEnumerator3
    UsePositionOffset=false
    Write Material Random=true
   )
   &SocketLocation(
    Orientation Sampling=::NewEnumerator1
    SamplePosOffset=true
    Sampled Position Offset=&MergeHelper_VectorSwitch(
     0=&VariantVectorSelector(
      Original Vector:Vector3f=<7 0 0>
      Variant 1 Vector:Vector3f=<0 0 0>
      Variant 2 Vector:Vector3f=&MergeHelper_VectorSwitch(
       0=<-31 -10.8 14.2>
       1=<-31 9.5 14.2>
       2=<-31 -10.9 9.4>
       3=<-31 9.5 9.4>
       4=<-31 -0.5 16>
       Index=&SequentialInt(Use Array Data:NiagaraBool=false, First Number:NiagaraInt32=0, Sequence:NiagaraDataInterfaceArrayInt32=?, Last Number:NiagaraInt32=4)
       Num=::NewEnumerator3
      )
      Variant 3 Vector:Vector3f=&MergeHelper_VectorSwitch(
       0=<-16 6.3 20.5>
       1=<-16 -8.3 20.5>
       2=<-16.5 -1 21>
       3=?
       Index=&SequentialInt(Use Array Data:NiagaraBool=false, First Number:NiagaraInt32=0, Sequence:NiagaraDataInterfaceArrayInt32=?, Last Number:NiagaraInt32=2)
       Num=::NewEnumerator1
      )
      Variant Amount:NiagaraInt32=2
     )
     1=&MergeHelper_VectorSwitch(
      0=<0 0 0>
      1=<0 0 0>
      2=&MergeHelper_VectorSwitch(
       0=<-31 -10.8 14.2>
       1=<-31 9.5 14.2>
       2=<-31 -10.9 9.4>
       3=<-31 9.5 9.4>
       4=<-31 -0.5 16>
       Index=&SequentialInt(Use Array Data:NiagaraBool=false, First Number:NiagaraInt32=0, Sequence:NiagaraDataInterfaceArrayInt32=?, Last Number:NiagaraInt32=4)
       Num=::NewEnumerator3
      )
      3=&MergeHelper_VectorSwitch(
       0=<-16 6.3 20.5>
       1=<-16 -8.3 20.5>
       2=<-16.5 -1 21>
       3=?
       Index=&SequentialInt(Use Array Data:NiagaraBool=false, First Number:NiagaraInt32=0, Sequence:NiagaraDataInterfaceArrayInt32=?, Last Number:NiagaraInt32=2)
       Num=::NewEnumerator1
      )
      Index=&MakeIntFromFloat(Float=$User.Variant)
      Num=::NewEnumerator2
     )
     2=<0 24 0>
     3=?
     Index=$Particles.VisibilityTag
     Num=::NewEnumerator1
    )
    Socket Index=/Niagara/DynamicInputs/Helpers/MaskInt(
     Mask Source:NiagaraInt32=1
     Mask Parameter:NiagaraInt32=$Particles.VisibilityTag
     Mask:NiagaraInt32=2
     Masked Int:NiagaraInt32=4
     Passthrough Int:NiagaraInt32=&SequentialInt(Use Array Data:NiagaraBool=false, First Number:NiagaraInt32=0, Sequence:NiagaraDataInterfaceArrayInt32=?, Last Number:NiagaraInt32=1)
    )
    Socket Reader=$System.Socket Reader
    Socket Sampling Mode=::NewEnumerator3
    VelocitySpeedLimit=false
   )
   &ShapeLocation(
    Hemisphere Distribution=<0.5 1>
    MaskSpawnGroup=false
    Override Seed=false
    Sphere Radius=&MergeHelper_FloatSwitch(0=0, 1=1, 2=0.5, 3=?, Index=$Particles.VisibilityTag, Num=::NewEnumerator1)
    UseOwnerScale=true
   )
   &AddVelocity(
    Module.Use Falloff Distance=false
    Override Seed=false
    Velocity Mode=::NewEnumerator1
    Velocity Origin=$Output.SocketLocation.SampledPosition
    Velocity Speed=&MergeHelper_FloatSwitch(
     0=0
     1=&RandomRangeFloat(Minimum=10, Maximum=50)
     2=&RandomRangeFloat(Minimum=15, Maximum=50)
     3=?
     Index=$Particles.VisibilityTag
     Num=::NewEnumerator1
    )
   ) {module=AddVelocity001}
   &AddVelocity(Velocity=&MergeHelper_VectorSwitch(
     0=<0 -20 0>
     1=&RandomRangeVector(Minimum=<0 -375 0>, Maximum=<0 -450 0>, Evaluation Type=::NewEnumerator1)
     2=&RandomRangeVector(Minimum=<-50 100 20>, Maximum=<50 200 100>, Evaluation Type=::NewEnumerator1)
     3=?
     Index=$Particles.VisibilityTag
     Num=::NewEnumerator1
    ))
   &InitialMeshOrientation(
    Mesh Orientation Mode=::NewEnumerator7
    Rotation=&MakeVector(
     X=&Add_Float(
      A=&RandomRangeFloat(Minimum=0, Maximum=1)
      B=&Lerp_Float(
       A=0
       B=3.14
       Alpha=&MakeFloatFromInt(Int=&SequentialInt(Use Array Data:NiagaraBool=false, First Number:NiagaraInt32=0, Sequence:NiagaraDataInterfaceArrayInt32=?, Last Number:NiagaraInt32=1))
      )
     )
     Y=0
     Z=0
    )
    Use Rotation Vector=true
   )
   &DynamicMaterialParameters(Module.Index0Float0Write=true, Module.Index0Float1Write=true, Module.Index0Float2Write=true, Module.Index0Float3Write=true)
   &CameraOffset(Camera Offset Amount=2)
  ]
  particleUpdateScript [
   &ParticleState()
   &ScaleMeshSize(Scale Factor=/Niagara/DynamicInputs/CascadeConversion/SplitVector(
     Double Component Vector Channels:ECascadeNiagaraTwoVectorChannels=::NewEnumerator1
     Double Component:NiagaraFloat=&FloatFromCurve(FloatCurve=DataInterface() {propertyValues {Curve {keys [
          {arriveTangent=-0.0252, time=0, value=1}
          {arriveTangent=-0.5041, interpMode=RCIM_Cubic, leaveTangent=-0.5041, tangentMode=RCTM_User, time=0.1035, value=0.9748}
          {arriveTangent=-1.0929, interpMode=RCIM_Cubic, leaveTangent=-5.7267, tangentMode=RCTM_Break, time=0.5983, value=0.9176}
          <at 0.7558 -0.153>
         ]}}})
     Single Component:NiagaraFloat=&FloatFromCurve(FloatCurve=DataInterface() {propertyValues {Curve {keys [
          {arriveTangent=-0.0038, time=0, value=1}
          {arriveTangent=0.3986, interpMode=RCIM_Cubic, leaveTangent=0.3986, tangentMode=RCTM_User, time=0.1035, value=0.9962}
          {arriveTangent=-0.087, interpMode=RCIM_Cubic, leaveTangent=-10.0837, tangentMode=RCTM_Break, time=0.5983, value=0.9919}
          <at 0.7558 -0.1629>
         ]}}})
    ))
   &ScaleSpriteSize(
    Scale Factor=&MaskVector2d(
     Mask Source:NiagaraInt32=1
     Mask Parameter:NiagaraInt32=$Particles.VisibilityTag
     Mask:NiagaraInt32=2
     Masked Vector2D:Vector2f=&Vector2DFromFloat(Value=&FloatFromCurve(FloatCurve=DataInterface() {propertyValues {Curve {keys [
           {arriveTangent=55.4408, interpMode=RCIM_Cubic, leaveTangent=55.4408, tangentMode=RCTM_User, time=0, value=0}
           {interpMode=RCIM_Cubic, tangentMode=RCTM_User, time=0.03, value=1}
           {arriveTangent=-3.3237, leaveTangent=-3.3237, time=1, value=0}
          ]}}}))
     Passthrough Vector2D:Vector2f=&Vector2DFromCurve(
      Vector2Curve=DataInterface() {propertyValues {
        XCurve {keys [{arriveTangent=-0.0316, time=-0.0016, value=0.6559}, <at 0.5953 0.6243>]}
        YCurve {keys [{arriveTangent=-0.417, time=-0.0016, value=0.6444}, <at 0.5953 0.2274>]}
       }}
      ScaleCurve=<1 1>
     )
    )
    Scale Sprite Size Mode=::NewEnumerator2
   )
   &Drag(
    Drag=&MaskFloat(Mask Source=1, Mask Parameter=$Particles.VisibilityTag, Mask=2, Masked Float=2, Passthrough Float=1)
    Use Linear Drag=true
    Use Rotational Drag=true
   )
   &ScaleColor(
    Scale Alpha=&FloatFromCurve(FloatCurve=DataInterface() {propertyValues {Curve {keys [<at 0 1>,<at 1 0>]}}})
    Scale RGB=&VectorFromCurve(VectorCurve=DataInterface() {propertyValues {XCurve {keys [<at 0.7 10>,<at 1 1>]}, YCurve {keys [<at 0.7 0.5>,<at 1 0>]}, ZCurve {keys [<at 0.7 0.3>,<at 1 0>]}}})
    ScaleA=true
    ScaleRGB=true
   )
   &FN_Sprite_Attributes(
    Color=&MultiplyLinearColor(
     Linear Color=<rgba 8 1.3634 0.2451 1>
     Scale Factor=&MakeVector4FromVectorAndFloat(
      "Vector (RGB)"=&VectorFromCurve(VectorCurve=DataInterface() {propertyValues {
         XCurve {keys [<at 0.35 2.0083 RCIM_Cubic>,<at 1 1 RCIM_Cubic>]}
         YCurve {keys [<at 0.35 2.7602 RCIM_Cubic>,<at 1 0.1346 RCIM_Cubic>]}
         ZCurve {keys [<at 0.35 4 RCIM_Cubic>,<at 1 0.0208 RCIM_Cubic>]}
        }})
      "Float (Alpha)"=&FloatFromCurve(FloatCurve=DataInterface() {propertyValues {Curve {keys [{arriveTangent=-6.8352, interpMode=RCIM_Cubic, leaveTangent=-6.8352, tangentMode=RCTM_User, time=0, value=2}, <at 0.2 1>]}}})
     )
    )
    Create Color Attribute=true
    Module Output:ENiagara_AttributeSamplingApplyOutput=::NewEnumerator1
   )
   &FN_Sprite_Attributes(
    Color=&MultiplyLinearColor(
     Linear Color=<rgba 1 1 1 1>
     Scale Factor=&MakeVector4FromVectorAndFloat(
      "Vector (RGB)"=&VectorFromCurve(
       VectorCurve=DataInterface() {propertyValues {XCurve {keys [<at 0 5>,<at 1 1>]}, YCurve {keys [<at 0 1>,<at 1 0.0773>]}, ZCurve {keys [<at 0 0>,<at 1 0>]}}}
       Scale Curve=<2.5 2.5 2.5>
      )
      "Float (Alpha)"=&FloatFromCurve(FloatCurve=DataInterface() {propertyValues {Curve {keys [<at 0 1 RCIM_Cubic>,<at 1 0 RCIM_Cubic>]}}})
     )
    )
    Create Color Attribute=true
    Module Output:ENiagara_AttributeSamplingApplyOutput=::NewEnumerator1
   ) {module=FN_Sprite_Attributes001}
   &AccelerationForce(Acceleration=&MaskVector(Mask Source=1, Mask Parameter=$Particles.VisibilityTag, Mask=2, Masked Vector=<0 0 -100>, Passthrough Vector=<0 0 0>))
   &SolveForcesAndVelocity(Clamp Velocity=false, Limit Acceleration=false)
   &ScaleSpriteSizeBySpeed(Max Scale Factor=<0.2 6>, Min Scale Factor=<0.2 1>)
  ]
  rendererClasses [&NiagaraMeshRendererProperties, &NiagaraSpriteRendererProperties]
  renderers [
   {rendererClass=&NiagaraMeshRendererProperties, rendererIndex=0}
   {rendererClass=&NiagaraSpriteRendererProperties, rendererIndex=1}
   {rendererClass=&NiagaraSpriteRendererProperties, rendererIndex=2}
  ]
 }
}
summary {
 systemName=NS_Pickaxe_Rhino_IdleSmoke
 emitters [
  {
   emitterName=SwingParticles_Local
   rendererClasses [&NiagaraMeshRendererProperties, &NiagaraSpriteRendererProperties]
  }
  {emitterName=SmokeLoop_Packed, rendererClasses [&NiagaraSpriteRendererProperties]}
  {
   emitterName=Fire_Packed
   rendererClasses [
    &NiagaraSpriteRendererProperties
    &NiagaraMeshRendererProperties
    &NiagaraLightRendererProperties
   ]
  }
 ]
 userVariables [
  User.bIsSwinging {type=NiagaraBool}
  User.Variant {type=NiagaraFloat, defaultValue=0, description=""}
 ]
}
