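# Niagara swing trail effect for the Rhino pickaxe, with selectors driving per-variant trail count, width, color and spawning.
# MainTrail builds anim-trail ribbons between socket pairs, then widens, lifts and fades them through curves and dynamic material parameters.
# Particles_Swing emits hot orange, speed-stretched sparks at sockets by distance moved, scattered by curl noise and drag.
# A variant-gated burst sets a visibility tag, alongside a fading light renderer positioned at a separate socket.
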NiagaraSystem /BRCosmetics/Weapons/FORT_Melee/Pickaxe_Rhino/FX/NS_Pickaxe_Rhino_Trail
dependencies {
 usedDataInterfaces [
  &NiagaraDataInterfaceCurve
  &NiagaraDataInterfaceVectorCurve
  &NiagaraDataInterfaceVector2DCurve
  &NiagaraDataInterfaceSocketReader
 ]
 usedDynamicInputs [&RandomRangeVector2D]
 usedModules [
  &self:SystemSpawnScript.SystemScriptSource.SystemScriptGraph.NiagaraNodeAssignment_0.SetVariables_#1
  &SystemState
  &FN_VariantManager
 ]
}
emitters {
 MainTrail {
  emitterSpawnScript [&FN_RibbonIDOffsetUtility()]
  emitterUpdateScript [
   &EmitterState()
   &FN_SpawnAnimTrail(Number of Anim Trails:NiagaraInt32=&MakeIntFromFloat(Float=/BRCosmetics/Weapons/FORT_Melee/Pickaxe_GateHound/FX/VariantManager/VariantFloatSelector(
      Original Float:NiagaraFloat=1
      Variant 1 Float:NiagaraFloat=1
      Variant 2 Float:NiagaraFloat=2
      Variant 3 Float:NiagaraFloat=1
      Variant Amount:NiagaraInt32=2
     )))
  ]
  particleSpawnScript [
   &FN_InitializeAnimTrailBoneIndecies(First Bone or Socket Index=1, Second Bone or Socket Index=0)
   &FN_InitializeAnimTrailBoneIndecies(AnimTrail Selection=1, First Bone or Socket Index=3, Second Bone or Socket Index=2) {module=FN_InitializeAnimTrailBoneIndecies001}
   &FN_InitializeAnimTrail(
    AnimTrailWidth=&VariantFloatSelector(
     Original Float:NiagaraFloat=0.8
     Variant 1 Float:NiagaraFloat=0.6
     Variant 2 Float:NiagaraFloat=0.5
     Variant 3 Float:NiagaraFloat=0.8
     Variant Amount:NiagaraInt32=2
    )
    Color=&VariantColorSelector(
     Variant Amount=2
     Original Color=<rgba 0.068 0.065 0.066 1>
     Variant 1 Color=<rgba 0.526 0.526 0.526 0.6>
     Variant 2 Color=<rgba 0.526 0.526 0.526 0.6>
     Variant 3 Color=<rgba 0.526 0.526 0.526 0.6>
    )
    Lifetime=0.9
   )
   /SetVariables(Particles.RibbonID=/Niagara/DynamicInputs/NiagaraID/MakeNiagaraIDFromInt(Index:NiagaraInt32=$Emitter.RibbonIDOffset, Acquire Tag:NiagaraInt32=$Particles.SpawnGroup)) {isSetParameters=true}
  ]
  particleUpdateScript [
   &FN_UpdateAnimTrail() {module=FN_UpdateAnimTrail001}
   &ScaleRibbonWidth(Ribbon Width Scale=&FloatFromCurve(FloatCurve=DataInterface() {propertyValues {Curve {keys [
         {
          arriveTangent=3.0404
          arriveTangentWeight=0.0019
          interpMode=RCIM_Cubic
          leaveTangent=3.0404
          leaveTangentWeight=0.3315
          tangentMode=RCTM_User
          time=0
          value=1.4516
         }
         <at 1 4.3956>
        ]}}}))
   &AccelerationForce(Acceleration=<0 0 50>)
   &ScaleColor(
    Scale Alpha=&FloatFromCurve(FloatCurve=DataInterface() {propertyValues {Curve {keys [
         {arriveTangent=0.0518, interpMode=RCIM_Cubic, leaveTangent=0.0518, tangentMode=RCTM_User, time=0, value=1}
         {arriveTangent=-0.3503, leaveTangent=-0.3503, time=1, value=0}
        ]}}})
    Scale RGB=/BRCosmetics/Weapons/FORT_Melee/Pickaxe_GateHound/FX/VariantManager/VariantVectorSelector(
     Original Vector:Vector3f=&VectorFromCurve(VectorCurve=DataInterface() {propertyValues {
        XCurve {keys [<at 0 5>,<at 0.1 0.401>,<at 1 0.3974>]}
        YCurve {keys [<at 0 3.7068>,<at 0.1 0.401>,<at 1 0.4062>]}
        ZCurve {keys [<at 0 2.2396>,<at 0.1 0.401>,<at 1 0.3805>]}
       }})
     Variant 1 Vector:Vector3f=&VectorFromCurve(VectorCurve=DataInterface() {propertyValues {
        XCurve {keys [
          {arriveTangent=-99.2917, time=0, value=100}
          {arriveTangent=-0.4125, leaveTangent=-0.4125, time=0.07, value=0.7083}
          {arriveTangent=-0.2218, leaveTangent=-0.2218, time=0.15, value=0.2958}
          <at 0.7 0.074>
         ]}
        YCurve {keys [
          {arriveTangent=-0.3125, time=0, value=1}
          {arriveTangent=-0.3646, leaveTangent=-0.3646, time=0.07, value=0.6875}
          {arriveTangent=-0.2396, leaveTangent=-0.2396, time=0.15, value=0.3229}
          <at 0.7 0.0833>
         ]}
        ZCurve {keys [
          {arriveTangent=0.6653, time=0, value=0}
          {arriveTangent=-0.3558, leaveTangent=-0.3558, time=0.07, value=0.6653}
          {arriveTangent=-0.2295, leaveTangent=-0.2295, time=0.15, value=0.3095}
          <at 0.7 0.0799>
         ]}
       }})
     Variant 2 Vector:Vector3f=&VectorFromCurve(VectorCurve=DataInterface() {propertyValues {
        XCurve {keys [
          {arriveTangent=-99.2917, time=0, value=100}
          {arriveTangent=-0.4125, leaveTangent=-0.4125, time=0.07, value=0.7083}
          {arriveTangent=-0.2218, leaveTangent=-0.2218, time=0.15, value=0.2958}
          <at 0.7 0.074>
         ]}
        YCurve {keys [
          {arriveTangent=-0.3125, time=0, value=1}
          {arriveTangent=-0.3646, leaveTangent=-0.3646, time=0.07, value=0.6875}
          {arriveTangent=-0.2396, leaveTangent=-0.2396, time=0.15, value=0.3229}
          <at 0.7 0.0833>
         ]}
        ZCurve {keys [
          {arriveTangent=0.6653, time=0, value=0}
          {arriveTangent=-0.3558, leaveTangent=-0.3558, time=0.07, value=0.6653}
          {arriveTangent=-0.2295, leaveTangent=-0.2295, time=0.15, value=0.3095}
          <at 0.7 0.0799>
         ]}
       }})
     Variant 3 Vector:Vector3f=&VectorFromCurve(VectorCurve=DataInterface() {propertyValues {
        XCurve {keys [
          {arriveTangent=-99.2917, time=0, value=100}
          {arriveTangent=-0.4125, leaveTangent=-0.4125, time=0.07, value=0.7083}
          {arriveTangent=-0.2218, leaveTangent=-0.2218, time=0.15, value=0.2958}
          <at 0.7 0.074>
         ]}
        YCurve {keys [
          {arriveTangent=-0.3125, time=0, value=1}
          {arriveTangent=-0.3646, leaveTangent=-0.3646, time=0.07, value=0.6875}
          {arriveTangent=-0.2396, leaveTangent=-0.2396, time=0.15, value=0.3229}
          <at 0.7 0.0833>
         ]}
        ZCurve {keys [
          {arriveTangent=0.6653, time=0, value=0}
          {arriveTangent=-0.3558, leaveTangent=-0.3558, time=0.07, value=0.6653}
          {arriveTangent=-0.2295, leaveTangent=-0.2295, time=0.15, value=0.3095}
          <at 0.7 0.0799>
         ]}
       }})
     Variant Amount:NiagaraInt32=2
    )
    ScaleA=true
    ScaleRGB=true
   )
   &DynamicMaterialParameters(
    Index 0 Param 1=&FloatFromCurve(FloatCurve=DataInterface() {propertyValues {Curve {keys [{arriveTangent=-2.8011, interpMode=RCIM_Cubic, leaveTangent=-2.8011, tangentMode=RCTM_User, time=0, value=1}, <at 1 0>]}}})
    Index 0 Param 3=0.2
    Index 0 Param 4=&FloatFromCurve(FloatCurve=DataInterface() {propertyValues {Curve {keys [{arriveTangent=-2.7093, interpMode=RCIM_Cubic, leaveTangent=-2.7093, tangentMode=RCTM_User, time=0, value=1}, <at 1 0>]}}})
    Module.Index0Float0Write=true
    Module.Index0Float1Write=true
    Module.Index0Float2Write=true
    Module.Index0Float3Write=true
   )
   /SetVariables(Particles.ScaledRibbonWidth:NiagaraFloat=&Multiply_Float(
     A=$Particles.Initial.RibbonWidth
     B=&FloatFromCurve(
      FloatCurve=DataInterface() {propertyValues {Curve {keys [{arriveTangent=3.0404, interpMode=RCIM_Cubic, leaveTangent=3.0404, tangentMode=RCTM_User, time=0.0056, value=1.4516}, <at 0.9928 4.3956>]}}}
      Scale Curve=1.2
     )
    )) {isSetParameters=true}
   &FN_Sprite_Attributes(
    Color=&MultiplyLinearColor(
     Linear Color=<rgba 0.641 0.6127 0.6221 1.6>
     Scale Factor=&MakeVector4FromVectorAndFloat(
      "Vector (RGB)"=&VectorFromCurve(VectorCurve=DataInterface() {propertyValues {
         XCurve {keys [<at 0 0.2917>,<at 0.1 0.2552>,<at 1 0.3974>]}
         YCurve {keys [<at 0 0.2917>,<at 0.1 0.2552>,<at 1 0.4062>]}
         ZCurve {keys [<at 0 0.2917>,<at 0.1 0.2552>,<at 1 0.3805>]}
        }})
      "Float (Alpha)"=&FloatFromCurve(FloatCurve=DataInterface() {propertyValues {Curve {keys [
           {arriveTangent=0.0518, interpMode=RCIM_Cubic, leaveTangent=0.0518, tangentMode=RCTM_User, time=0, value=1}
           {arriveTangent=-0.3503, leaveTangent=-0.3503, time=1, value=0}
          ]}}})
     )
    )
    Create Color Attribute=true
    Module Output:ENiagara_AttributeSamplingApplyOutput=::NewEnumerator1
   )
   &SolveForcesAndVelocity(Clamp Velocity=false, Limit Acceleration=false)
  ]
  rendererClasses [&NiagaraRibbonRendererProperties]
  renderers [
   {rendererClass=&NiagaraRibbonRendererProperties}
   {rendererClass=&NiagaraRibbonRendererProperties, rendererIndex=1}
  ]
 }
 Particles_Swing {
  emitterUpdateScript [
   &EmitterState(Life Cycle Mode=::NewEnumerator1, Loop Behavior=::NewEnumerator1, Loop Duration=0.3, UseLoopDelay=false)
   &SpawnPerUnit(
    Max Movement Threshold=2500
    Module.Use Max Movement Threshold=false
    Module.Use Movement Tolerance=false
    Spawn Probability:NiagaraFloat=&FloatFromCurve(
     FloatCurve=DataInterface() {propertyValues {Curve {keys [<at 0 0 RCIM_Constant>,<at 0.05 1 RCIM_Constant>]}}}
     CurveIndex=$Emitter.NormalizedLoopAge
    )
    Spawn Spacing=15
    Use Spawn Probability=true
   )
   &SpawnBurst_Instantaneous(
    Spawn Group=1
    Spawn Probability=&VariantFloatSelector(
     Original Float:NiagaraFloat=1
     Variant 1 Float:NiagaraFloat=0
     Variant 2 Float:NiagaraFloat=1
     Variant 3 Float:NiagaraFloat=1
     Variant Amount:NiagaraInt32=2
    )
    Use Spawn Probability=true
    UseLoopCountLimit=false
   )
  ]
  particleSpawnScript [
   /SetVariables(Particles.VisibilityTag=$Emitter.SpawnGroup) {isSetParameters=true}
   &InitializeParticle(
    Lifetime=&RandomRangeFloat(Minimum=0.17, Maximum=0.435, Evaluation Type=::NewEnumerator1)
    Sprite Rotation Mode=::NewEnumerator2
    Sprite Size Mode=::NewEnumerator3
    Uniform Sprite Size=&RandomRangeFloat(Minimum=2.35, Maximum=4.5)
    UsePositionOffset=false
    Write Material Random=true
   )
   &SocketLocation(
    Interpolate Socket Sampling=true
    SamplePosOffset=false
    Socket Index=&MakeIntFromFloat(Float=&VariantFloatSelector(
      Original Float:NiagaraFloat=1
      Variant 1 Float:NiagaraFloat=1
      Variant 2 Float:NiagaraFloat=&UniformAOrBFloat(A=1, B=3, Evaluation Type=::NewEnumerator0, Recalculate Random Each Loop=true, Override Seed=false)
      Variant 3 Float:NiagaraFloat=1
      Variant Amount:NiagaraInt32=2
     ))
    Socket Reader=$System.Socket Reader
    Socket Sampling Mode=::NewEnumerator3
    VelocitySpeedLimit=false
   )
   &ShapeLocation(MaskSpawnGroup=false, Override Seed=false, Sphere Radius=8, UseOwnerScale=true)
   &AddVelocity(
    Module.Use Falloff Distance=false
    Override Seed=false
    Velocity Mode=::NewEnumerator1
    Velocity Origin=$Output.SocketLocation.SampledPosition
    Velocity Speed=&RandomRangeFloat(Minimum=5, Maximum=35)
   )
   &InheritVelocity(Inherited Velocity Amount Scale=<-1 -1 -1>, Inherited Velocity Speed Limit=150, Speed Limit=true, UseSpeedThreshold=false)
  ]
  particleUpdateScript [
   &ParticleState()
   &ScaleSpriteSize(
    Scale Factor=&Vector2DFromCurve(
     Vector2Curve=DataInterface() {propertyValues {
       XCurve {keys [{arriveTangent=-0.0316, time=-0.0016, value=0.6559}, <at 0.5953 0.6243>]}
       YCurve {keys [{arriveTangent=-0.417, time=-0.0016, value=0.6444}, <at 0.5953 0.2274>]}
      }}
     ScaleCurve=<1 1>
    )
    Scale Sprite Size Mode=::NewEnumerator2
   )
   &CurlNoiseForce(Bias Noise=false, Noise Frequency=25, Noise Strength=750, Pan Noise=false)
   &AccelerationForce(Acceleration=<0 0 -100>)
   &Drag(Drag=7, Use Linear Drag=true, Use Rotational Drag=true)
   &ScaleColor(
    Scale Mode=1
    Scale RGBA=&MakeVector4_FromLinearColor(LinearColor=&MultiplyLinearColor(
      Linear Color=<rgba 8 1.3634 0.2451 1>
      Scale Factor=&MakeVector4FromVectorAndFloat(
       "Vector (RGB)"=&VectorFromCurve(
        VectorCurve=DataInterface() {propertyValues {
          XCurve {keys [<at 0.35 2.0083 RCIM_Cubic>,<at 1 1 RCIM_Cubic>]}
          YCurve {keys [<at 0.35 2.7602 RCIM_Cubic>,<at 1 0.1346 RCIM_Cubic>]}
          ZCurve {keys [<at 0.35 4 RCIM_Cubic>,<at 1 0.0208 RCIM_Cubic>]}
         }}
        Scale Curve=&VectorFromFloat(Value=&FloatFromCurve(FloatCurve=DataInterface() {propertyValues {Curve {keys [{arriveTangent=-26.9617, interpMode=RCIM_Cubic, time=0, value=2}, <at 1 0 RCIM_Cubic>]}}}))
       )
       "Float (Alpha)"=1
      )
     ))
    ScaleRGBA=true
   )
   &SocketLocation(
    Orientation Sampling=::NewEnumerator1
    Position Sampling=::NewEnumerator1
    SamplePosOffset=false
    Socket Index=4
    Socket Reader=$System.Socket Reader
    Socket Sampling Mode=::NewEnumerator3
    VelocitySpeedLimit=false
   ) {module=SocketLocation001}
   &FN_Light_Attributes(
    Color=<rgba 150 31.7648 0 1>
    Create Color Attribute=true
    Create Exponent Attribute=true
    Create Position Attribute=true
    Create Radius Attribute=true
    Exponent=4
    Position:NiagaraPosition=$Output.SocketLocation001.SampledPosition
    Radius=125
    Scale Color=&FloatFromCurve(FloatCurve=DataInterface() {propertyValues {Curve {keys [{arriveTangent=-1.7649, interpMode=RCIM_Cubic, leaveTangent=-1.7649, tangentMode=RCTM_User, time=0, value=1}, <at 1 0 RCIM_Cubic>]}}})
   )
   &SolveForcesAndVelocity(Clamp Velocity=false, Limit Acceleration=false)
   &ScaleSpriteSizeBySpeed(Max Scale Factor=<0.5 15>, Min Scale Factor=<0.5 1>)
  ]
  rendererClasses [&NiagaraSpriteRendererProperties, &NiagaraLightRendererProperties]
  renderers [
   {rendererClass=&NiagaraSpriteRendererProperties}
   {rendererClass=&NiagaraLightRendererProperties, rendererIndex=1}
  ]
 }
}
summary {
 systemName=NS_Pickaxe_Rhino_Trail
 emitters [
  {emitterName=MainTrail, rendererClasses [&NiagaraRibbonRendererProperties]}
  {
   emitterName=Particles_Swing
   rendererClasses [&NiagaraSpriteRendererProperties, &NiagaraLightRendererProperties]
  }
 ]
 userVariables [
  User.RibbonIDOffset {type=NiagaraInt32}
  User.Variant {type=NiagaraFloat, defaultValue=0, description=""}
 ]
}
