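# Physics asset for the Rhino pickaxe, a two-handed melee weapon mesh, defining its simplified collision and jointed physics.
# The Melee_TwoHanded_Handle body pairs a long capsule with a large box near the upper end, likely shaft and head.
# A small tilted box body on dyn_handle connects to it through one constraint that locks both swing axes, leaving twist as the axis of rotation.
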PhysicsAsset /BRCosmetics/Weapons/FORT_Melee/Pickaxe_Rhino/Meshes/Pickaxe_Rhino_Physics
bodies {
 Melee_TwoHanded_Handle [
  {shapeType=Capsule, length=101.6502, radius=3.0748, center=<0.9136 -0.9262 30.6051>, rotation=<rot -0.1935 -89.9538 0.3089>}
  {shapeType=Box, extentX=21.6916, extentY=66.0644, extentZ=18.7551, center=<1.5086 3.3643 81.7076>}
 ]
 dyn_handle [{shapeType=Box, extentX=22.4141, extentY=15.0435, extentZ=3.2717, center=<0 -5.0828 5.0827>, rotation=<rot 0 0 44.9998>}]
}
constraints [{
  bone1Name=dyn_handle
  bone2Name=Melee_TwoHanded_Handle
  swing1LimitDegrees=45
  swing1Motion=Locked
  swing2LimitDegrees=45
  swing2Motion=Locked
  twistLimitDegrees=15
 }]
